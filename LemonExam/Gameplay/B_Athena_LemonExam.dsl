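; Weapon Blueprint for the LemonExam melee weapon, granting a custom sprint modifier and sprint-jump fall-damage immunity while equipped.
; Equipping registers the sprint class, listens for the SprintJumping tag to apply the immunity effect, and starts cosmetic cue and Niagara ambience; unequip and EndPlay handle cleanup.
; Appear, disappear and heavy-attack visuals are time-stamped material parameter animations.
; Building impacts sample the Scene component's position 0.08 seconds apart to orient a fading slash decal.
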(event FortWeapon|EventOnEquippedAndAttached
  (bind _returnvalue (Game|GetInstigator))
  (bind _asfort_player_pawn_athena (Utilities|Casting|CastToFortPlayerPawnAthena _returnvalue))
  (bind _instigator (Variables|Actor|GetInstigator))
  ;[0.05381 0.15 0.1098 0.5] On Equip
  (|CallParentFunction)
  (Utilities|FlowControl|Sequence
    (:then_0
      (Variables|Default|SetPlayerPawn _asfort_player_pawn_athena)
      (Variables|Default|SetCharacterMesh (Variables|/Script/Engine.Character|GetMesh _asfort_player_pawn_athena))
      (CallFunction|PlayAppearTA 0.0) ; Reset to hide
      (CallFunction|PlayAppearTA 1.0 0.5) ; Play appear
      (bind _output_get (Variables|Default|SetTactical%20Sprint%20Component (Actor|GetComponentbyClass (Game|GetInstigatorController) "/Script/TacticalSprintRuntime.FortPlayerControllerComponent_TacticalSprint")))
      (Sprint|AddCustomSprintMMEClass _output_get (Variables|Default|GetSprint%20MMEClass%20to%20Add))
      (bind _asynctaskproxy (Ability|Tasks|WaitGameplayTagAddToActor _instigator "(TagName=\"Gameplay.Action.Player.SprintJumping\")")
        (:then
          (Variables|Transient|SetAsyncAction_WaitSprintJump _asynctaskproxy))
        (:Added
          (Fortnite|ApplyGameplayEffectsToActor _instigator (Utilities|Array|MakeArray (Variables|Settings|GetFallDamageImmunityGameplayEffectClass)) (0.0 0.0 0.0) false 0 (Ability|GetFortAbilitySystemComponent _instigator) (GameplayEffects|MakeEffectContext (Ability|GetFortAbilitySystemComponent _instigator))))))
    (:then_1
      (Networking|CanExecuteCosmeticEvents
        (:True
          (if (not (Vehicle|IsPassengerinVehicle (Variables|Default|GetPlayerPawn)))
            (GameplayCue|AddGameplayCueOnActor(Looping) self (Variables|Default|GetGameplay%20Cue%20Tag) (Ability|GameplayCue|MakeGameplayCueParameters 0.0))
            (Components|Activation|Activate (Variables|Default|GetNS_LemonExam_Ambient) false)))))))

(event FortWeapon|EventOnUnEquippedAndDetached
  ;[0.15 0.07841 0.01641 0.5] Cleanup: On Un-Equip + End Play
  (|CallParentFunction)
  (Networking|CanExecuteCosmeticEvents
    (:True
      (GameplayCue|RemoveGameplayCueOnActor(Looping) self (Variables|Default|GetGameplay%20Cue%20Tag) (Ability|GameplayCue|MakeGameplayCueParameters 0.0))
      (Components|Activation|Deactivate (Variables|Default|GetNS_LemonExam_Ambient)))))

(event FortWeapon|EventOnLocalUnEquip
  ;[0.15 0.07841 0.01641 0.5] Cleanup: On Un-Equip + End Play
  (|CallParentFunction)
  (CallFunction|PlayAppearTA 2.0 0.5)
  (Sprint|RemoveCustomSprintMMEClass (Variables|Default|GetTactical%20Sprint%20Component) (Variables|Default|GetSprint%20MMEClass%20to%20Add))
  (bind _asyncaction_waitsprintjump (Variables|Transient|GetAsyncAction_WaitSprintJump)
    (:then
      (Ability|Async|EndAction _asyncaction_waitsprintjump))
    (:else)))

(event EventEndPlay
  (params (EndPlayReason :type EEndPlayReason%20Enum :default "Destroyed"))
  ;[0.15 0.07841 0.01641 0.5] Cleanup: On Un-Equip + End Play
  (|CallParentFunction EndPlayReason)
  (if (not (and (> (ScalableFloat|GetCachedScalableFloatValue (Variables|Default|GetFixup%20Sprint%20MME%20Removal)) 0) (not (FortWeapon|IsEquipped))))
    (CallFunction|PlayAppearTA 2.0 0.5)
    (Sprint|RemoveCustomSprintMMEClass (Variables|Default|GetTactical%20Sprint%20Component) (Variables|Default|GetSprint%20MMEClass%20to%20Add)))
  (bind _asyncaction_waitsprintjump (Variables|Transient|GetAsyncAction_WaitSprintJump)
    (:then
      (Ability|Async|EndAction _asyncaction_waitsprintjump))
    (:else)))

(event Custom|PlayHeavyAttack_TA
  (bind _weaponmesh (Variables|Visual|GetWeaponMesh))
  ; PlayHeavyAttack_TA
  (Rendering|Material|SetScalarParameterValueonMaterials _weaponmesh "ATK_AnimationDuration" 2.0) ; Set Duration. 2s
  (Rendering|Material|SetScalarParameterValueonMaterials _weaponmesh "ATK_Stage" 1.0) ; Stage1
  (Rendering|Material|SetScalarParameterValueonMaterials _weaponmesh "ATK_GameTimeFromBP" (Utilities|Time|GetGameTimeinSeconds)) ; Passthough Time
)

(event Custom|PlayAppear_TA
  (params (Stage :type Float%20%28double-precision%29) (Duration :type Float%20%28single-precision%29))
  (bind _weaponmesh (Variables|Visual|GetWeaponMesh))
  ; PlayEquip_TA
  (Rendering|Material|SetScalarParameterValueonMaterials _weaponmesh "APR_AnimationDuration" Duration) ; Set Duration. 1s
  (Rendering|Material|SetScalarParameterValueonMaterials _weaponmesh "APR_Stage" Stage) ; Stage\n0:Inital (Hide).\n1:Appear.\n2:Disappear
  (Rendering|Material|SetScalarParameterValueonMaterials _weaponmesh "APR_GameTimeFromBP" (Utilities|Time|GetGameTimeinSeconds)) ; Passthough Time
)

(event FortWeapon|EventOnPlayImpactFX
  (params (HitResult :type Hit%20Result%20Structure%20%28by%20ref%29) (ImpactPhysicalSurface :type EPhysicalSurface%20Enum :default "SurfaceType_Default") (SpawnedPSC :type FXSystem%20Component%20Object%20Reference))
  (bind _scene (Variables|Default|GetScene))
  (bind _returnvalue (Transformation|GetWorldLocation _scene))
  (bind (_bblockinghit _binitialoverlap _time _distance _location _impactpoint _normal _impactnormal _physmat _hitactor) (Collision|BreakHitResult HitResult))
  (bind (_asbuilding_actor _bsuccess) (Utilities|Casting|CastToBuildingActor _hitactor))
  ;[0.0745 0.01908 0.15 0.5] Directional Slash Decal set
  (|CallParentFunction HitResult ImpactPhysicalSurface SpawnedPSC)
  (if _bsuccess
    (Variables|Default|SetImpactTipLocation _returnvalue)
    (Utilities|FlowControl|Delay 0.08)
    (Variables|Default|SetDecalDirection (Math|Vector|Normalize (- _returnvalue (Variables|Default|GetImpactTipLocation))))
    (bind _returnvalue_2 (Rendering|Decal|SpawnDecalatLocation "/LemonExam/Assets/FX/Attack123/MI_LemonExam_Melee_Decal.MI_LemonExam_Melee_Decal" "32.000000,98.000000,6.000000" _impactpoint (Math|Rotator|MakeRotfromXY _impactnormal (Variables|Default|GetDecalDirection)) 2.0))
    (Rendering|Components|Decal|SetFadeOut _returnvalue_2 0.0 2.0 true)))

(fn ConstructionScript
  (|CallParentFunction)
  (Rendering|Material|SetMaterial (Variables|Visual|GetWeaponMesh) 0 (Class|BTechArtGlobalLibrary|GetDynamicMaterialInstance (Variables|Visual|GetWeaponMesh))))
