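; Gameplay ability for the LemonExam weapon (Rumi's sword) that plays a 2D pickup music stinger when the sword is added to the inventory.
; Gated by a hotfix toggle, it checks the avatar pawn's ability system for a persistent gameplay effect marking the stinger as heard.
; If absent, it applies that effect, then plays the stinger unless in a spatial gameplay experience.
; CanActivateAbility always returns true.
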(event Ability|EventActivateAbility
  (bind _addtoinventorystingergameplayeffectclass (Variables|Settings|GetAddToInventoryStingerGameplayEffectClass))
  ;[0.001541 0.1412 0.15 0.5] ActivateAbility: This ability is activated when rumi's sword is added to the inventory. We're using a persistant gameplay effect to determine if the player needs to hear the stringer sound.
  (if (ScalableFloat|GetValueasBool (Variables|Hotfix|GetHotfix_PlayWeaponPickupStinger))
    (bind _outfortplayerpawn (CallFunction|GetAvatarasValidFortPlayerPawn)
      (:ValidFortPlayerPawn
        (Utilities|IsValid (Ability|GetFortAbilitySystemComponent _outfortplayerpawn)
          (:Is%20Valid
            (if (<= (GameplayEffects|GetGameplayEffectCount (Ability|GetFortAbilitySystemComponent _outfortplayerpawn) _addtoinventorystingergameplayeffectclass 0 true) 0)
              (Ability|ApplyGameplayEffectToOwner _addtoinventorystingergameplayeffectclass)
              (if (not (Fortnite|IsSpatialGameplayExperience))
                (Audio|PlaySound2D "/LemonExam/Assets/Audio/MSS_LemonExam_Pickup_MX.MSS_LemonExam_Pickup_MX")))))))))

(fn CanActivateAbility
  (params (ActorInfo :type Gameplay%20Ability%20Actor%20Info%20Structure) (Handle :type Gameplay%20Ability%20Spec%20Handle%20Structure))
  (return 0 true))
